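# Niagara system for the DecoRed vehicle boost cosmetic, throwing short-lived sprite particles from boost sockets on a skeletal mesh.
# Its single triangles emitter partitions particles across ports, with spawn rate varying for frontend preview and lifetime switched by game mode.
# Particles fly along negative X, shed inherited source movement and fall under increasing downward acceleration.
# Five sprite renderers draw them, with spawn group as renderer tag and curve-driven purple, pink and yellow colors; paint override is disabled.
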NiagaraSystem /VehicleCosmetics/Boosts/Boost_DecoRed/VFX/NS_Boost_DecoRed
dependencies {
 usedDataInterfaces [
  &NiagaraDataInterfaceSpriteRendererInfo
  &NiagaraDataInterfaceCurve
  &NiagaraDataInterfaceVectorCurve
  &NiagaraDataInterfaceColorCurve
  &NiagaraDataInterfaceSkeletalMesh
 ]
 usedDynamicInputs []
 usedModules [
  &self:SystemSpawnScript.SystemScriptSource.SystemScriptGraph.NiagaraNodeAssignment_0.SetVariables_#1
  &SystemState
 ]
}
emitters {NE_Part_Triangles {
  emitterUpdateScript [
   &EmitterState()
   &SpawnRate(SpawnGroup=1, SpawnRate=6, Use Spawn Probability=false) {module=SpawnRate001}
   &VC_FilteredSocketCount()
   &SpawnRate(
    SpawnRate=&MakeCustomFloatFromBool(Bool=$User.bIsFrontendPreview, True Float=14, False Float=12)
    Use Spawn Probability=false
   )
  ]
  particleSpawnScript [
   &VC_PaintedVariantsOverride() {enabled=false}
   &InitializeParticle(
    Color Mode=::NewEnumerator0
    Lifetime=&MergeHelper_FloatSwitch(0=0.5, 1=0.5, 2=0.5, 3=?, Index=$User.GameModeID, Num=::NewEnumerator1)
    Position Mode=::NewEnumerator0
    Sprite Rotation Mode=::NewEnumerator2
    Sprite Size Max=<60 60>
    Sprite Size Min=<80 80>
    Sprite Size Mode=::NewEnumerator2
    Write Material Random=true
   )
   &PartitionParticles(Number Of Partitions=$Emitter.PortCount)
   &SkeletalMeshLocation(
    Bone / Socket Index=$Particles.PartitionParticles.Partition
    Bone / Socket Sampling Mode=::NewEnumerator5
    MaskSpawnGroup=false
    Mesh Sampling Type=::NewEnumerator6
    Sampled Position Offset=&AddVectorToPosition(Position=<0 0 0>, Vector=<-50 0 0>)
    Skeletal Mesh=$System.SkeletalMesh
   )
   &AddVelocity(Velocity=&MakeCustomVectorFromBool(
     Bool=$User.bIsFrontendPreview
     True Vector=&RandomRangeVector(Minimum=<-600 200 0>, Maximum=<-800 -200 50>)
     False Vector=&RandomRangeVector(Minimum=<-600 200 0>, Maximum=<-800 -200 50>)
    ))
   &ShapeLocation(MaskSpawnGroup=false, Override Seed=false, Sphere Radius=30, UseOwnerScale=true)
   &Color(Color=<rgba 0.0262 0.0027 0.1413 1>)
   &SubUVAnimation(
    "Sprite Renderer "=DataInterface() {propertyValues {SpriteRenderer=&self:NE_Part_Triangles_0.NiagaraSpriteRendererProperties_1}}
    SubUV Animation Mode:ENiagara_SubUVLookupModeV2=::NewEnumerator2
    UseEndFrame=false
    UseStartFrame=false
   )
   /SetVariables(Particles.PinkColor:LinearColor=<rgba 1 0.0003 0.2195 1>, Particles.RendererTag:NiagaraInt32=&SpawnGroup()) {isSetParameters=true}
  ]
  particleUpdateScript [
   &ParticleState()
   &InheritSourceMovement(
    Applied Position Scale=&Multiply_VectorByFloat(
     Vector=<1 1 1>
     Float=&FloatFromCurve(FloatCurve=DataInterface() {propertyValues {Curve {keys [<at 0 1>,<at 1 0>]}}})
    )
    ApplyForce=false
    ApplyPosition=true
    ApplyPositionFalloff=false
   )
   &AccelerationForce(Acceleration=&VectorFromCurve(
     VectorCurve=DataInterface() {propertyValues {XCurve {keys [<at 0 0>,<at 1 0>]}, YCurve {keys [<at 0 0>,<at 1 0>]}, ZCurve {keys [<at 0 0 RCIM_Cubic>,<at 1 -1 RCIM_Cubic>]}}}
     Scale Curve=<1 1 2000>
    ))
   &SolveForcesAndVelocity(Clamp Velocity=false, Limit Acceleration=false)
   /SetVariables(
    Particles.PinkOffsetColor:LinearColor=&ColorFromCurve(ColorCurve=DataInterface() {propertyValues {
       AlphaCurve {keys [<at 0.2003 1 RCIM_Cubic>,<at 0.7164 0 RCIM_Cubic>]}
       BlueCurve {keys [<at 0 0.0561 RCIM_Cubic>]}
       GreenCurve {keys [<at 0 0.0006 RCIM_Cubic>]}
       RedCurve {keys [<at 0 0.807 RCIM_Cubic>]}
      }})
    Particles.PurpleColor:LinearColor=&ColorFromCurve(ColorCurve=DataInterface() {propertyValues {
       AlphaCurve {keys [<at 0 1>,<at 0.6042 1>,<at 0.9402 0>]}
       BlueCurve {keys [<at 0.0684 0.2812>,<at 0.7326 0.1195>]}
       GreenCurve {keys [<at 0.0684 0.0057>,<at 0.7326 0.0212>]}
       RedCurve {keys [<at 0.0684 0.0416>,<at 0.7326 0.016>]}
      }})
    Particles.YellowOutline:LinearColor=&ColorFromCurve(ColorCurve=DataInterface() {propertyValues {
       AlphaCurve {keys [<at 0.2627 1 RCIM_Cubic>,<at 0.8674 0 RCIM_Cubic>]}
       BlueCurve {keys [<at 0 0.0561 RCIM_Cubic>]}
       GreenCurve {keys [<at 0 0.0006 RCIM_Cubic>]}
       RedCurve {keys [<at 0 0.807 RCIM_Cubic>]}
      }})
    Particles.Yellow_Color:LinearColor=&ColorFromCurve(ColorCurve=DataInterface() {propertyValues {
       AlphaCurve {keys [<at 0.2353 1 RCIM_Cubic>,<at 0.8782 0 RCIM_Cubic>]}
       BlueCurve {keys [<at 0 0.0044 RCIM_Cubic>]}
       GreenCurve {keys [<at 0 0.3613 RCIM_Cubic>]}
       RedCurve {keys [<at 0 0.807 RCIM_Cubic>]}
      }})
   ) {isSetParameters=true}
  ]
  rendererClasses [&NiagaraSpriteRendererProperties]
  renderers [
   {rendererClass=&NiagaraSpriteRendererProperties}
   {rendererClass=&NiagaraSpriteRendererProperties, rendererIndex=1}
   {rendererClass=&NiagaraSpriteRendererProperties, rendererIndex=2}
   {rendererClass=&NiagaraSpriteRendererProperties, rendererIndex=3}
   {rendererClass=&NiagaraSpriteRendererProperties, rendererIndex=4}
  ]
 }}
summary {
 systemName=NS_Boost_DecoRed
 emitters [{emitterName=NE_Part_Triangles, rendererClasses [&NiagaraSpriteRendererProperties]}]
 userVariables [
  User.bIsFrontendPreview {type=NiagaraBool}
  User.bIsPainted {type=NiagaraBool}
  User.ColorVariantID {type=NiagaraInt32}
  User.GameModeID {type=NiagaraInt32}
  User.PaintedColor {type=LinearColor, description="", defaultValue {struct=LinearColor, value=<rgba 1 1 1 1>}}
  User.PaintedScalar {type=NiagaraFloat, defaultValue=1, description=""}
  User.PortCount {type=NiagaraInt32, defaultValue=1}
 ]
}
