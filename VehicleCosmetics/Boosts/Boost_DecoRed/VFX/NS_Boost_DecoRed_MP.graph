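# Niagara system for the DecoRed vehicle boost cosmetic, trailing pink, purple, blue and yellow sprites and pink ribbons behind the car.
# Two triangle sprite emitters partition particles across vehicle sockets; the main one tags spawn groups for six renderers.
# A ribbon emitter spawns one trail per port, with width and color driven by curves.
# All three inherit source movement and pick lifetime by game mode; painted-variant overrides are disabled.
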NiagaraSystem /VehicleCosmetics/Boosts/Boost_DecoRed/VFX/NS_Boost_DecoRed_MP
dependencies {
 usedDataInterfaces [
  &NiagaraDataInterfaceSpriteRendererInfo
  &NiagaraDataInterfaceCurve
  &NiagaraDataInterfaceVectorCurve
  &NiagaraDataInterfaceColorCurve
  &NiagaraDataInterfaceSkeletalMesh
 ]
 usedDynamicInputs []
 usedModules [
  &self:SystemSpawnScript.SystemScriptSource.SystemScriptGraph.NiagaraNodeAssignment_0.SetVariables_#1
  &SystemState
 ]
}
emitters {
 NE_Part_Triangles {
  emitterUpdateScript [
   &EmitterState()
   &SpawnRate(SpawnGroup=2, SpawnRate=12, Use Spawn Probability=false) {module=SpawnRate002}
   &SpawnRate(SpawnGroup=1, SpawnRate=10, Use Spawn Probability=false) {module=SpawnRate001}
   &VC_FilteredSocketCount()
   &SpawnRate(SpawnRate=50, Use Spawn Probability=false)
  ]
  particleSpawnScript [
   &VC_PaintedVariantsOverride() {enabled=false}
   &InitializeParticle(
    Color Mode=::NewEnumerator0
    Lifetime=&MergeHelper_FloatSwitch(0=0.5, 1=0.5, 2=0.5, 3=?, Index=$User.GameModeID, Num=::NewEnumerator1)
    Position Mode=::NewEnumerator0
    Sprite Rotation Mode=::NewEnumerator2
    Sprite Size Max=<60 60>
    Sprite Size Min=<80 80>
    Sprite Size Mode=::NewEnumerator2
    Write Material Random=true
   )
   &PartitionParticles(Number Of Partitions=$Emitter.PortCount)
   &SkeletalMeshLocation(
    Bone / Socket Index=$Particles.PartitionParticles.Partition
    Bone / Socket Sampling Mode=::NewEnumerator5
    MaskSpawnGroup=false
    Mesh Sampling Type=::NewEnumerator6
    Sampled Position Offset=&AddVectorToPosition(Position=<0 0 0>, Vector=<-50 0 0>)
    Skeletal Mesh=$System.SkeletalMesh
   )
   &AddVelocity(Velocity=&MakeCustomVectorFromBool(
     Bool=$User.bIsFrontendPreview
     True Vector=&RandomRangeVector(Minimum=<-900 100 0>, Maximum=<-1150 -100 50>)
     False Vector=&RandomRangeVector(Minimum=<-900 100 0>, Maximum=<-1150 -100 50>)
    ))
   &ShapeLocation(MaskSpawnGroup=false, Override Seed=false, Sphere Radius=45, UseOwnerScale=true)
   &SubUVAnimation(
    "Sprite Renderer "=DataInterface() {propertyValues {SpriteRenderer=&self:NE_Part_Triangles_0.NiagaraSpriteRendererProperties_1}}
    SubUV Animation Mode:ENiagara_SubUVLookupModeV2=::NewEnumerator2
    UseEndFrame=false
    UseStartFrame=false
   )
   /SetVariables(Particles.RendererTag:NiagaraInt32=&SpawnGroup()) {isSetParameters=true, module=SetVariables_#2}
  ]
  particleUpdateScript [
   &ParticleState()
   &ScaleSpriteSize(
    Uniform Curve Scale=1.5
    Uniform Curve Sprite Scale=DataInterface() {propertyValues {Curve {keys [
        {interpMode=RCIM_Cubic, tangentMode=RCTM_Break, time=0.0024, value=0.3722}
        {arriveTangent=2.7292, interpMode=RCIM_Cubic, leaveTangent=2.7292, time=0.1277, value=0.4711}
        {interpMode=RCIM_Cubic, tangentMode=RCTM_Break, time=0.2216, value=0.9704}
       ]}}}
   )
   &InheritSourceMovement(
    Applied Position Scale=&Multiply_VectorByFloat(Vector=<1 1 1>, Float=1)
    ApplyForce=false
    ApplyPosition=true
    ApplyPositionFalloff=false
   )
   &AccelerationForce(Acceleration=&VectorFromCurve(
     VectorCurve=DataInterface() {propertyValues {XCurve {keys [<at 0 0>,<at 1 0>]}, YCurve {keys [<at 0 0>,<at 1 0>]}, ZCurve {keys [<at 0 0 RCIM_Cubic>,<at 1 -1 RCIM_Cubic>]}}}
     Scale Curve=<1 1 2000>
    )) {enabled=false}
   &SolveForcesAndVelocity(Clamp Velocity=false, Limit Acceleration=false)
   /SetVariables(
    Particles.BueTransparent:LinearColor=&ColorFromCurve(ColorCurve=DataInterface() {propertyValues {
       AlphaCurve {keys [<at 0 1>,<at 0.6782 0>]}
       BlueCurve {keys [<at 0.0456 0.6354>,<at 0.64 0.1195>]}
       GreenCurve {keys [<at 0.0456 0.1525>,<at 0.64 0.0212>]}
       RedCurve {keys [<at 0.0456 0.1065>,<at 0.64 0.016>]}
      }})
    Particles.PinkOffsetColor:LinearColor=&ColorFromCurve(ColorCurve=DataInterface() {propertyValues {
       AlphaCurve {keys [<at 0.2003 1 RCIM_Cubic>,<at 0.7164 0 RCIM_Cubic>]}
       BlueCurve {keys [<at 0 0.0561 RCIM_Cubic>]}
       GreenCurve {keys [<at 0 0.0006 RCIM_Cubic>]}
       RedCurve {keys [<at 0 0.807 RCIM_Cubic>]}
      }})
    Particles.PinkOutline:LinearColor=&ColorFromCurve(ColorCurve=DataInterface() {propertyValues {
       AlphaCurve {keys [<at 0.2186 1 RCIM_Cubic>,<at 0.8183 0 RCIM_Cubic>]}
       BlueCurve {keys [<at 0 0.1779 RCIM_Cubic>]}
       GreenCurve {keys [<at 0 0.011 RCIM_Cubic>]}
       RedCurve {keys [<at 0 0.7529 RCIM_Cubic>]}
      }})
    Particles.PurpleColor:LinearColor=&ColorFromCurve(ColorCurve=DataInterface() {propertyValues {
       AlphaCurve {keys [<at 0 1>,<at 0.6042 1>,<at 0.9402 0>]}
       BlueCurve {keys [<at 0.0684 0.2812>,<at 0.7326 0.1195>]}
       GreenCurve {keys [<at 0.0684 0.0057>,<at 0.7326 0.0212>]}
       RedCurve {keys [<at 0.0684 0.0416>,<at 0.7326 0.016>]}
      }})
    Particles.YellowOutline:LinearColor=&ColorFromCurve(ColorCurve=DataInterface() {propertyValues {
       AlphaCurve {keys [<at 0.2627 1 RCIM_Cubic>,<at 0.8674 0 RCIM_Cubic>]}
       BlueCurve {keys [<at 0 0.0561 RCIM_Cubic>]}
       GreenCurve {keys [<at 0 0.0006 RCIM_Cubic>]}
       RedCurve {keys [<at 0 0.807 RCIM_Cubic>]}
      }})
    Particles.YellowSize:Vector2f=&RandomRangeVector2D(Minimum=<45 45>, Maximum=<85 85>)
    Particles.Yellow_Color:LinearColor=&ColorFromCurve(ColorCurve=DataInterface() {propertyValues {
       AlphaCurve {keys [<at 0.2353 1 RCIM_Cubic>,<at 0.7651 0 RCIM_Cubic>]}
       BlueCurve {keys [<at 0 0.0044 RCIM_Cubic>]}
       GreenCurve {keys [<at 0 0.3613 RCIM_Cubic>]}
       RedCurve {keys [<at 0 0.807 RCIM_Cubic>]}
      }})
   ) {isSetParameters=true, module=SetVariables_#3}
  ]
  rendererClasses [&NiagaraSpriteRendererProperties]
  renderers [
   {rendererClass=&NiagaraSpriteRendererProperties}
   {rendererClass=&NiagaraSpriteRendererProperties, rendererIndex=1}
   {rendererClass=&NiagaraSpriteRendererProperties, rendererIndex=2}
   {rendererClass=&NiagaraSpriteRendererProperties, rendererIndex=3}
   {rendererClass=&NiagaraSpriteRendererProperties, rendererIndex=4}
   {rendererClass=&NiagaraSpriteRendererProperties, rendererIndex=5}
  ]
 }
 NE_Part_Triangles001 {
  emitterUpdateScript [
   &EmitterState()
   &VC_FilteredSocketCount()
   &SpawnRate(SpawnRate=20, Use Spawn Probability=false)
  ]
  particleSpawnScript [
   &VC_PaintedVariantsOverride() {enabled=false}
   &InitializeParticle(
    Color Mode=::NewEnumerator0
    Lifetime=&MergeHelper_FloatSwitch(0=0.35, 1=0.35, 2=0.35, 3=?, Index=$User.GameModeID, Num=::NewEnumerator1)
    Position Mode=::NewEnumerator0
    Sprite Rotation Mode=::NewEnumerator2
    Sprite Size Max=<55 55>
    Sprite Size Min=<35 35>
    Sprite Size Mode=::NewEnumerator2
    Write Material Random=true
   )
   &PartitionParticles(Number Of Partitions=$Emitter.PortCount)
   &SkeletalMeshLocation(
    Bone / Socket Index=$Particles.PartitionParticles.Partition
    Bone / Socket Sampling Mode=::NewEnumerator5
    MaskSpawnGroup=false
    Mesh Sampling Type=::NewEnumerator6
    Sampled Position Offset=&AddVectorToPosition(Position=<0 0 0>, Vector=<-50 0 0>)
    Skeletal Mesh=$System.SkeletalMesh
   )
   &AddVelocity(Velocity=/VehicleCosmetics/FX/Niagara/Modules/MergeEmitters/VC_VectorBySpawnGroup(
     SpawnGroup or ExecIndex:NiagaraInt32=0
     CustomInt:NiagaraInt32=0
     MaximumSpawnGroups:NiagaraInt32=2
     Vector Group 0:Vector3f=&RandomRangeVector(Minimum=<-500 -70 -120>, Maximum=<-500 70 200>)
     Vector Group 1:Vector3f=&RandomRangeVector(Minimum=<-500 -70 -120>, Maximum=<-500 70 200>)
     Vector Group 2:Vector3f=<0 0 0>
     Vector Group 3:Vector3f=<0 0 0>
     Vector Group 4:Vector3f=<0 0 0>
     Vector Group 5:Vector3f=<0 0 0>
    ))
   &ShapeLocation(MaskSpawnGroup=false, Override Seed=false, Sphere Radius=30, UseOwnerScale=true)
   &SubUVAnimation(
    "Sprite Renderer "=DataInterface() {propertyValues {SpriteRenderer=&self:NE_Part_Triangles001_0.NiagaraSpriteRendererProperties_1}}
    SubUV Animation Mode:ENiagara_SubUVLookupModeV2=::NewEnumerator2
    UseEndFrame=false
    UseStartFrame=false
   )
   /SetVariables(Particles.PinkColor:LinearColor=<rgba 1 0.0003 0.2195 1>, Particles.RendererTag:NiagaraInt32=&SpawnGroup()) {isSetParameters=true, module=SetVariables_#2}
  ]
  particleUpdateScript [
   &ParticleState()
   &InheritSourceMovement(
    Applied Position Scale=&Multiply_VectorByFloat(
     Vector=<1 1 1>
     Float=&FloatFromCurve(FloatCurve=DataInterface() {propertyValues {Curve {keys [<at 0 1>,<at 1 0>]}}})
    )
    ApplyForce=false
    ApplyPosition=true
    ApplyPositionFalloff=false
   )
   &SolveForcesAndVelocity(Clamp Velocity=false, Limit Acceleration=false)
   /SetVariables(
    Particles.BueTransparent:LinearColor=&ColorFromCurve(ColorCurve=DataInterface() {propertyValues {
       AlphaCurve {keys [<at 0 1>,<at 0.6782 0>]}
       BlueCurve {keys [<at 0.0456 0.3958>,<at 0.64 0.1195>]}
       GreenCurve {keys [<at 0.0456 0.0572>,<at 0.64 0.0212>]}
       RedCurve {keys [<at 0.0456 0.0392>,<at 0.64 0.016>]}
      }})
    Particles.PinkOffsetColor:LinearColor=&ColorFromCurve(ColorCurve=DataInterface() {propertyValues {
       AlphaCurve {keys [<at 0.2003 1 RCIM_Cubic>,<at 0.7164 0 RCIM_Cubic>]}
       BlueCurve {keys [<at 0 0.0561 RCIM_Cubic>]}
       GreenCurve {keys [<at 0 0.0006 RCIM_Cubic>]}
       RedCurve {keys [<at 0 0.807 RCIM_Cubic>]}
      }})
    Particles.PurpleColor:LinearColor=&ColorFromCurve(ColorCurve=DataInterface() {propertyValues {
       AlphaCurve {keys [<at 0 1>,<at 0.6042 1>,<at 0.9402 0>]}
       BlueCurve {keys [<at 0.0684 0.2812>,<at 0.7326 0.1195>]}
       GreenCurve {keys [<at 0.0684 0.0057>,<at 0.7326 0.0212>]}
       RedCurve {keys [<at 0.0684 0.0416>,<at 0.7326 0.016>]}
      }})
   ) {isSetParameters=true, module=SetVariables_#3}
  ]
  rendererClasses [&NiagaraSpriteRendererProperties]
  renderers [{rendererClass=&NiagaraSpriteRendererProperties}]
 }
 NE_VC_Boost_Ribbons {
  emitterUpdateScript [
   &EmitterState(Life Cycle Mode=::NewEnumerator1, Loop Duration=1, UseLoopDelay=false)
   &VC_FilteredSocketCount()
   &VC_RateQualityHandler(High Spawn Rate=60, Low Spawn Rate=60, Medium Spawn Rate=60, Use Custom Rate?=true)
   &VC_SpawnRibbon_v1(Number of Ribbons=$Emitter.PortCount, Spawn Time=1, SpawnRate=$Emitter.Spawn Rate)
  ]
  particleSpawnScript [
   &VC_PaintedVariantsOverride(Orange=<rgba 0.6724 0.0883 0.2836 1>, Use R1 Painted Overrides=true, Use R2 Painted Overrides=true) {enabled=false}
   &VC_InitializeRibbon_v1(
    Custom Offset=true
    Enable Skeletal Position=true
    Facing Axis:ENiagaraOrientationAxis=::YAxis
    Lifetime=&MergeHelper_FloatSwitch(0=0.75, 1=0.75, 2=0.75, 3=?, Index=$User.GameModeID, Num=::NewEnumerator1)
    Local Vector Offset=&AddVectorToPosition(Position=<20 0 -8>, Vector=$User.DTCarOffset)
    Offset Amount=1
    Ribbon Width=90
   )
   &AddVelocity(Velocity=<-1000 0 0>) {module=AddVelocity001}
  ]
  particleUpdateScript [
   &ParticleState()
   &InheritSourceMovement(ApplyForce=false, ApplyPosition=true, ApplyPositionFalloff=false) {module=InheritSourceMovement002}
   &SolveForcesAndVelocity(Clamp Velocity=false, Limit Acceleration=false)
   &VC_UpdateRibbon_v1(Enable Skeletal Position=true)
   /SetVariables(
    Particles.PinkColor:LinearColor=&ColorFromCurve(ColorCurve=DataInterface() {propertyValues {
       AlphaCurve {keys [
         <at 0.0255 0>
         {arriveTangent=0.1094, interpMode=RCIM_Cubic, leaveTangent=0.1094, time=0.0607, value=0.35}
         <at 0.4827 0.05>
         <at 0.734 0 RCIM_Cubic>
        ]}
       BlueCurve {keys [<at 0 0.117 RCIM_Cubic>]}
       GreenCurve {keys [<at 0 0.0015 RCIM_Cubic>]}
       RedCurve {keys [<at 0 0.8796 RCIM_Cubic>]}
      }})
    Particles.R1_Width:NiagaraFloat=&FloatFromCurve(
     FloatCurve=DataInterface() {propertyValues {Curve {keys [
         <at 0 0.2 RCIM_Cubic>
         {arriveTangent=0.9676, interpMode=RCIM_Cubic, leaveTangent=0.9676, time=0.15, value=1.1}
         {arriveTangent=-0.7845, interpMode=RCIM_Cubic, leaveTangent=-0.7845, tangentMode=RCTM_User, time=0.4976, value=0.6815}
         {interpMode=RCIM_Cubic, tangentMode=RCTM_Break, time=0.8, value=0.05}
        ]}}}
     Scale Curve=90
    )
   ) {isSetParameters=true}
  ]
  rendererClasses [&NiagaraRibbonRendererProperties]
  renderers [{rendererClass=&NiagaraRibbonRendererProperties}]
 }
}
summary {
 systemName=NS_Boost_DecoRed_MP
 emitters [
  {emitterName=NE_Part_Triangles, rendererClasses [&NiagaraSpriteRendererProperties]}
  {emitterName=NE_VC_Boost_Ribbons, rendererClasses [&NiagaraRibbonRendererProperties]}
  {emitterName=NE_Part_Triangles001, rendererClasses [&NiagaraSpriteRendererProperties]}
 ]
 userVariables [
  User.bIsFrontendPreview {type=NiagaraBool}
  User.bIsPainted {type=NiagaraBool}
  User.ColorVariantID {type=NiagaraInt32}
  User.DTCarOffset {type=Vector3f, description="", defaultValue {struct=Vector3f, value=<0 0 0>}}
  User.GameModeID {type=NiagaraInt32}
  User.PaintedColor {type=LinearColor, description="", defaultValue {struct=LinearColor, value=<rgba 1 1 1 1>}}
  User.PaintedScalar {type=NiagaraFloat, defaultValue=1, description=""}
  User.PortCount {type=NiagaraInt32, defaultValue=1}
 ]
}
